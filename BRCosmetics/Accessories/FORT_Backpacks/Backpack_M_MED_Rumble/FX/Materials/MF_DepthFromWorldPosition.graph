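# Material function that converts a world position into camera depth, stored with the Rumble backpack cosmetic's FX materials.
# It subtracts the camera position from the input position and takes the dot product with the camera direction vector, giving distance along the view axis rather than radial distance.
# The input defaults to world position excluding all shader offsets.
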MaterialFunction /BRCosmetics/Accessories/FORT_Backpacks/Backpack_M_MED_Rumble/FX/Materials/MF_DepthFromWorldPosition
graph {nodes [
  n1 = FunctionInput(Preview=WorldPosition() {worldPositionShaderOffset=WPT_ExcludeAllShaderOffsets}.XYZ) {bUsePreviewValueAsDefault=true, description=World Position to convert to Depth, inputName=World Position}
  n2 = CameraPositionWS()
  FunctionOutput(None=DotProduct(
    A=Subtract(A=@n1, B=@n2)
    B=MaterialFunctionCall() {materialFunction=&CameraDirectionVector}.Result
   )) {description=Depth from the input World Position, outputName=Depth}
 ]}
