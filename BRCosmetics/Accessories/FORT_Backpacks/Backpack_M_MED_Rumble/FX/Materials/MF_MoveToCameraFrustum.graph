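# Material function that outputs World Position Offset to move vertices onto a camera-facing plane in the view frustum, stored with Rumble backpack FX.
# XY input (default UV0) is centered, then scaled along camera right and up axes by depth, FOV and aspect ratio.
# Depth defaults just past the near clip, derived by Custom HLSL.
# Camera position and forward vector place the plane; subtracting world position yields the offset.
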MaterialFunction /BRCosmetics/Accessories/FORT_Backpacks/Backpack_M_MED_Rumble/FX/Materials/MF_MoveToCameraFrustum
graph {nodes [
  n1 = FunctionInput(Preview=TextureCoordinate() {desc="Using UVs here\r\nbut could be anything"}) {
   bUsePreviewValueAsDefault=true
   description=The XY position of the vertex in screen space. Defaults to UV0.
   inputName=XY View Position
   inputType=FunctionInput_Vector2
  }
  n2 = Subtract(A=@n1) {constB=0.5}
  n3 = ComponentMask(None=@n2) {r=true}
  n4 = Transform(None=Constant3Vector() {constant=<rgba 1 0 0 0>}) {transformSourceType=TRANSFORMSOURCE_Camera}
  n5 = ViewProperty() {property=MEVP_TanHalfFieldOfView}
  n6 = Custom() {
   code="(mul(float4(ResolvedView.WorldCameraOrigin.xyz,1),View.WorldToClip)).z"
   description=Near Clip Plane
   outputType=CMOT_Float1
   inputs []
  }
  n7 = FunctionInput(Preview=Add(A=@n6)) {
   bUsePreviewValueAsDefault=true
   description=The depth of the vertex in relation to the camera frustum. Defaults to the Near Clip Plane +1.
   inputName=Z View Depth
   inputType=FunctionInput_Scalar
   sortPriority=1
  }
  n8 = Multiply(A=Multiply(A=ComponentMask(None=@n5.Property) {r=true}) {constB=2}, B=@n7)
  n9 = ComponentMask(None=@n2) {g=true}
  n10 = Transform(None=Constant3Vector() {constant=<rgba 0 -1 0 0>}) {transformSourceType=TRANSFORMSOURCE_Camera}
  n11 = ViewSize()
  n12 = Divide(A=ComponentMask(None=@n11) {g=true}, B=ComponentMask(None=@n11) {r=true})
  n13 = Add(A=Multiply(A=Multiply(A=@n3, B=@n4), B=@n8), B=Multiply(A=Multiply(A=@n9, B=@n10), B=Multiply(A=@n8, B=@n12)))
  n14 = WorldPosition()
  n15 = Multiply(A=@n7, B=MaterialFunctionCall() {materialFunction=&CameraDirectionVector}.Result)
  n16 = CameraPositionWS()
  FunctionOutput(None=Add(A=Add(A=@n15, B=@n16), B=Subtract(A=@n13, B=@n14.XYZ)) {desc=to world position offset}) {description=Moves vertices to a specified point in the frustum, outputName=World Position Offset, sortPriority=1}
 ]}
