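# Material function for the Rumble backpack FX that outputs a screen UV offset producing stepped wiggle and vibration.
# Wiggle samples a torus flowmap with panned, scaled screen-aligned UVs centered on the object's position.
# Vibration samples the same flowmap at the object's panned clip-space position, giving a uniform per-object jitter.
# Both use floored time for choppy motion, center RG around zero, scale by intensity over the Depth input, and are summed.
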MaterialFunction /BRCosmetics/Accessories/FORT_Backpacks/Backpack_M_MED_Rumble/FX/Materials/MF_ScreenSpaceWiggle
graph {nodes [
  n1 = ObjectPositionWS()
  n2 = VectorParameter() {parameterName="UV Speed and Scale, Time Scale", defaultValue=<rgba -2.618 0.618 110.8 1.618>}
  n3 = Time()
  n4 = ScalarParameter() {defaultValue=13, parameterName=Time Steps}
  n5 = Floor(None=Multiply(A=@n3, B=@n4))
  n6 = Panner(
   Coordinate=Multiply(
    A=MaterialFunctionCall!(Center Point World Position=@n1) {materialFunction=/Engine/Functions/Engine_MaterialFunctions02/UVs/WorldSpaceAlignedScreenCoordinates}.Offset and Scaled UVS
    B=@n2.B
   )
   Time=Multiply(A=Divide(A=@n5, B=@n4), B=@n2.A)
   Speed=ComponentMask(None=@n2.RGB) {g=true, r=true}
  ) {bFractionalPart=true}
  n7 = Subtract(A=ComponentMask(None=TextureSample!(UVs=@n6, Tex, Level, Apply View MipBias) {mipValueMode=TMVM_MipLevel, samplerType=SAMPLERTYPE_LinearColor, texture=/Game/Global/Textures/Default/Flowmaps/T_TorusFlowmap}.RGB) {g=true, r=true}) {constB=0.5}
  n8 = FunctionInput() {inputName=Depth, inputType=FunctionInput_Scalar}
  n9 = Divide(A=ScalarParameter() {defaultValue=1, parameterName=Wiggle Intensity}, B=@n8)
  n10 = Divide(A=ScalarParameter() {defaultValue=2, parameterName=Vibrate Intensity}, B=@n8)
  n11 = ScalarParameter() {defaultValue=11, parameterName=Vibration Time Steps}
  n12 = Multiply(A=@n3, B=@n11)
  n13 = Panner(
   Coordinate=MaterialFunctionCall(World Position=ObjectPositionWS()) {materialFunction=&TransformToClipSpace}.Clip Space XY
   Time=Divide(A=Floor(None=@n12), B=@n11)
   Speed=AppendVector(
    A=ScalarParameter() {defaultValue=42.35, parameterName=Vibration Panner Speed U}
    B=ScalarParameter() {defaultValue=36.18, parameterName=Vibration Panner Speed V}
   )
  ) {bFractionalPart=true}
  n14 = Subtract(A=ComponentMask(None=TextureSample(UVs=@n13) {mipValueMode=TMVM_MipLevel, samplerType=SAMPLERTYPE_LinearColor, texture=&T_TorusFlowmap}.RGB) {g=true, r=true}) {constB=0.5}
  FunctionOutput(None=Add(A=Multiply(A=@n7, B=@n9), B=Multiply(A=@n10, B=@n14))) {outputName=Screen UVs Offset}
 ]}
