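# Material function for Rumble backpack FX that outputs a World Position Offset squeezing the mesh in camera depth.
# It feeds vertex screen position, plus wiggle scaled by a particle dynamic parameter, into MF_MoveToCameraFrustum at the object pivot's depth plus an offset input.
# Near Clip gates the offset and Mesh Squeeze scales it.
# A view-angle term from the instance's -Y axis then shifts the result along the camera ray by Back offset.
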MaterialFunction /BRCosmetics/Accessories/FORT_Backpacks/Backpack_M_MED_Rumble/FX/Materials/MF_Backpack_Rumble_MeshSqueeze
graph {nodes [
  n1 = TransformPosition(None=Constant3Vector() {constant=<rgba 0 0 33 0>}) {transformSourceType=TRANSFORMPOSSOURCE_View, transformType=TRANSFORMPOSSOURCE_World}
  n2 = Subtract(
   A=ScreenPosition().ViewportUV
   B=Multiply(
    A=MaterialFunctionCall(World Position=@n1) {materialFunction=&TransformToClipSpace}.Clip Space XY
    B=Constant2Vector() {g=-0.5, r=0.5}
   )
  )
  n3 = MaterialFunctionCall!(World Position=ObjectPositionWS()) {materialFunction=&/MF_DepthFromWorldPosition}
  n4 = Multiply(
   A=MaterialFunctionCall!(Depth=@n3.Depth) {materialFunction=&/MF_ScreenSpaceWiggle}.Screen UVs Offset
   B=DynamicParameter() {paramNames ["", Wiggle Intensity, "", ""]}.Wiggle Intensity
  )
  n5 = Add(A=@n3.Depth, B=FunctionInput() {bUsePreviewValueAsDefault=true, inputName=Depth Offset, inputType=FunctionInput_Scalar})
  n6 = Ceil(None=Subtract(A=@n5, B=ScalarParameter() {defaultValue=33, parameterName=Near Clip}))
  n7 = ScalarParameter() {defaultValue=0.99, parameterName=Mesh Squeeze}
  n8 = Multiply(
   A=MaterialFunctionCall!(XY View Position=Add(A=@n2, B=@n4) {constB=0}, Z View Depth=@n5) {materialFunction=&/MF_MoveToCameraFrustum}.World Position Offset
   B=Multiply(A=Saturate(None=@n6), B=Saturate(None=@n7))
  )
  n9 = MaterialFunctionCall() {materialFunction=&CameraDirectionVector}
  n10 = Transform(None=Constant3Vector() {constant=<rgba 0 -1 0 0>}) {transformSourceType=TRANSFORMSOURCE_Instance}
  n11 = Multiply(A=Saturate(None=DotProduct(A=@n9.Result, B=@n10)), B=ScalarParameter() {defaultValue=-9, parameterName=Back offset})
  n12 = WorldPosition()
  n13 = Normalize(VectorInput=Subtract(A=CameraPositionWS(), B=Add(A=@n12.XYZ, B=@n8)))
  FunctionOutput(None=Add(A=@n8, B=Multiply(A=@n11, B=@n13))) {outputName=World Position Offset}
 ]}
