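# M_Backpack_Rumble_Shock_Shell is an unlit, two-sided material for Niagara mesh particles, part of the Rumble backpack cosmetic's effects, apparently for a shock shell effect.
# Its graph centers on the MF_Backpack_Rumble_MeshSqueeze material function, fed by a Depth Offset scalar parameter with a negative default.
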Material /BRCosmetics/Accessories/FORT_Backpacks/Backpack_M_MED_Rumble/FX/Materials/M_Backpack_Rumble_Shock_Shell
graph {
 properties {bUsedWithNiagaraMeshParticles=true, pixelDepthOffsetMode=PDOM_Legacy, shadingModel=MSM_Unlit, twoSided=true, usedShadingModels=Unlit}
 nodes [MaterialFunctionCall!(Depth Offset=ScalarParameter() {defaultValue=-9, parameterName=Depth Offset}) {materialFunction=&/MF_Backpack_Rumble_MeshSqueeze}]
}
