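# Lit material for Niagara mesh particles that gives the Rumble backpack's shock skeleton a glowing flash effect.
# A particle-driven Glow Intensity scales a Fresnel glow and boosted eye mask from inverted specular-texture red, blending center and flash colors into emissive, with exponents packed in color alphas.
# Rising glow fades diffuse and specular texture contributions.
# A shared MeshSqueeze function takes a negative depth offset.
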Material /BRCosmetics/Accessories/FORT_Backpacks/Backpack_M_MED_Rumble/FX/Materials/M_Backpack_Rumble_Shock_Skeleton
graph {
 properties {bUsedWithNiagaraMeshParticles=true, pixelDepthOffsetMode=PDOM_Legacy, usedShadingModels=Default Lit}
 nodes [
  n1 = OneMinus(None=DotProduct(A=PixelNormalWS(), B=CameraVectorWS()))
  n2 = VectorParameter() {parameterName="Flash Color && Fresnel Exponent", defaultValue=<rgba 1 0.76 0.1 1>}
  n3 = DynamicParameter() {defaultValue=<rgba 1 0 0 1>, paramNames [Glow Intensity, - unused -, - unused -, - unused -]}
  n4 = Saturate(None=Multiply(A=@n3.Glow Intensity, B=ScalarParameter() {defaultValue=1, parameterName=Fresnel Glow Intensity}))
  n5 = Multiply(
   A=LinearInterpolate(
    A=ScalarParameter() {defaultValue=0.2, parameterName=Brightness Min}
    B=ScalarParameter() {defaultValue=3, parameterName=Brightness Max}
    Alpha=Power(Base=@n1, Exp=@n2.A)
   )
   B=@n4
  )
  n6 = TextureSample() {samplerType=SAMPLERTYPE_Masks, texture=/BRCosmetics/Accessories/FORT_Backpacks/Backpack_M_MED_Rumble/Texture/T_M_MED_Rumble_Backpack_S}
  n7 = Multiply(A=@n6.R) {constB=3.333}
  n8 = Multiply(A=OneMinus(None=Saturate(None=@n7)), B=Multiply(A=@n3.Glow Intensity, B=ScalarParameter() {defaultValue=9, parameterName=Eyes Mask Boost}))
  n9 = VectorParameter() {parameterName="Center Color && Color Lerp Exponent", defaultValue=<rgba 1 0.3052 0 0.5>}
  n10 = Saturate(None=Add(A=Power(Base=@n1, Exp=@n9.A), B=@n8))
  MaterialFunctionCall(Emissive=Multiply(A=Add(A=@n5, B=@n8), B=LinearInterpolate(A=@n9.RGB, B=@n2.RGB, Alpha=@n10))) {materialFunction=&MF_FX_EmissiveAdjustments}
  MaterialFunctionCall!(Depth Offset=ScalarParameter() {defaultValue=-12, parameterName=Depth Offset}) {materialFunction=&/MF_Backpack_Rumble_MeshSqueeze}
  n11 = OneMinus(None=@n4)
  Multiply(A=TextureSample() {texture=/BRCosmetics/Accessories/FORT_Backpacks/Backpack_M_MED_Rumble/Texture/T_M_MED_Rumble_Backpack_D}.RGB, B=@n11)
  Multiply(A=@n6.R, B=@n11)
  OneMinus(None=@n4)
  TextureSample() {samplerType=SAMPLERTYPE_Masks, texture=/BRCosmetics/Accessories/FORT_Backpacks/Backpack_M_MED_Rumble/Texture/T_M_MED_Rumble_Backpack_M}
  TextureSample() {samplerType=SAMPLERTYPE_Normal, texture=/BRCosmetics/Accessories/FORT_Backpacks/Backpack_M_MED_Rumble/Texture/T_M_MED_Rumble_Backpack_N}
 ]
}
