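# Niagara effect for the Rumble back bling, adding flashing electric shocks and flickering meshes to the backpack.
# FlashingElectricity loops at randomized intervals scaled by User.ShocksPerSecond, spawning short-lived light, sprite and mesh flashes that cycle five mesh indices by loop count.
# FlickeringMeshes, gated by User.bKillReactivity, toggles visibility tags from emitter age at MeshFlickersPerSecond and drives a stepped material pulse at the shock rate.
# A debug sprite emitter is disabled.
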NiagaraSystem /BRCosmetics/Accessories/FORT_Backpacks/Backpack_M_MED_Rumble/FX/NS_Backpack_Rumble
dependencies {usedDataInterfaces [&NiagaraDataInterfaceCurve], usedDynamicInputs [], usedModules [&SystemState]}
emitters {
 FN_Debug {
  bEnabled=false
  emitterUpdateScript [
   &EmitterState()
   &SpawnBurst_Instantaneous(Spawn Count=1, Use Spawn Probability=false)
  ]
  particleSpawnScript [/Niagara/Modules/Spawn/Initialization/InitializeParticle(
    Lifetime=5
    Module.Write Color=true
    Module.Write Lifetime=true
    Module.Write Mass=true
    Module.Write Material Random=true
    Module.Write Position=false
    Module.Write Scale=false
    Module.Write SpriteRotation=false
    Module.Write SpriteSize=true
    Position=&SimulationPosition()
    Sprite Size=<100 100>
   )]
  particleUpdateScript [
   &ParticleState()
   &FN_Timer()
   /SetVariables(Particles.Debug=0.1) {isSetParameters=true}
   &DynamicMaterialParameters(
    Index 0 Param 1=$Particles.Debug
    Module.Param0WriteEnabled=true
    Module.Param1WriteEnabled=true
    Module.Param2WriteEnabled=true
    Module.Param3WriteEnabled=true
   )
   &CameraOffset(Camera Offset Amount=111)
  ]
  rendererClasses [&NiagaraSpriteRendererProperties]
  renderers [{rendererClass=&NiagaraSpriteRendererProperties}]
 }
 FlashingElectricity {
  emitterUpdateScript [
   &EmitterState(
    Life Cycle Mode=::NewEnumerator1
    Loop Duration=&Divide_Float(
     A=&MakeCustomFloatFromBool(Bool=&RandomBool(Recalculate Random Each Loop=true), True Float=1, False Float=0.5)
     B=$User.ShocksPerSecond
    )
    Recalculate Duration Each Loop=true
    UseLoopDelay=false
   )
   &SpawnBurst_Instantaneous(Use Spawn Probability=false, UseLoopCountLimit=false)
  ]
  particleSpawnScript [
   /Niagara/Modules/Spawn/Initialization/V2/InitializeParticle(
    Color=<rgba 0.1 0.052 0.01 1>
    Lifetime=&Divide_Float(
     A=&MakeCustomFloatFromBool(Bool=&RandomBool(), True Float=0.5, False Float=0.333)
     B=$User.ShocksPerSecond
    )
    Position Offset=<0 -11 -7>
    Position Offset Coordinate Space=::Local
    Sprite Size Mode=::NewEnumerator3
    Uniform Sprite Size=66
    UsePositionOffset=true
    Write Material Random=true
   )
   /SetVariables(Particles.GlowCameraOffset=11, Particles.MeshIndex=&ModuloInt(A=$Emitter.LoopCount, B=5)) {isSetParameters=true}
  ]
  particleUpdateScript [
   &ParticleState()
   &FN_Light_Attributes(Color=<rgba 11 6.776 0.44 1>, Create Color Attribute=true)
  ]
  rendererClasses [
   &NiagaraLightRendererProperties
   &NiagaraMeshRendererProperties
   &NiagaraSpriteRendererProperties
  ]
  renderers [
   {rendererClass=&NiagaraLightRendererProperties, rendererIndex=0}
   {rendererClass=&NiagaraMeshRendererProperties, rendererIndex=1}
   {rendererClass=&NiagaraSpriteRendererProperties, rendererIndex=2}
  ]
 }
 FlickeringMeshes {
  emitterUpdateScript [
   &EmitterState(
    Inactive Response=::NewEnumerator1
    Life Cycle Mode=::NewEnumerator1
    Loop Behavior=::NewEnumerator1
    Loop Duration Mode=::NewEnumerator1
    UseLoopDelay=false
   )
   &SpawnBurst_Instantaneous(
    Spawn Probability=&MakeCustomFloatFromBool(Bool=$User.bKillReactivity, True Float=1, False Float=0)
    Use Spawn Probability=true
    UseLoopCountLimit=false
   )
  ]
  particleSpawnScript [&InitializeParticle(
    Color Mode=::NewEnumerator0
    Position Offset=<0 2 0>
    Sprite Size Mode=::NewEnumerator3
    Uniform Sprite Size=33
    UsePositionOffset=true
    Write Material Random=true
   )]
  particleUpdateScript [
   &ParticleState(
    Kill Particles When Lifetime Has Elapsed=false
    Let Infinitely Lived Particles Die When Emitter Deactivates=true
    Loop Particles Lifetime=true
   )
   /SetVariables(Particles.VisibilityTag=&MakeIntFromFloat(Float=/Niagara/DynamicInputs/Integer/V2/Round_Float(Float:NiagaraFloat=&Frac_Float(Float=&Multiply_Float(
        A=&Add_Float(
         A=$Emitter.Age
         B=&MakeCustomFloatFromBool(Bool=&RandomBool(Randomness Mode=::NewEnumerator2, Random Seed=?), True Float=0.5, False Float=0)
        )
        B=$User.MeshFlickersPerSecond
       ))))) {isSetParameters=true}
   &DynamicMaterialParameters(
    Index 0 Param 1=&FloatFromCurve(
     FloatCurve=DataInterface() {propertyValues {Curve {keys [<at 0 1 RCIM_Constant>,<at 0.2 0.5 RCIM_Constant>,<at 0.2333 0 RCIM_Constant>]}}}
     CurveIndex=&Frac_Float(Float=&Multiply_Float(A=$Emitter.Age, B=$User.ShocksPerSecond))
    )
    Index 0 Param 2=1
    Module.Param0WriteEnabled=true
    Module.Param1WriteEnabled=true
    Module.Param2WriteEnabled=true
    Module.Param3WriteEnabled=true
   )
  ]
  rendererClasses [&NiagaraMeshRendererProperties]
  renderers [
   {rendererClass=&NiagaraMeshRendererProperties, rendererIndex=0}
   {rendererClass=&NiagaraMeshRendererProperties, rendererIndex=1}
   {rendererClass=&NiagaraMeshRendererProperties, rendererIndex=2}
  ]
 }
}
summary {
 systemName=NS_Backpack_Rumble
 emitters [
  {emitterName=FlickeringMeshes, rendererClasses [&NiagaraMeshRendererProperties]}
  {bEnabled=false, emitterName=FN_Debug, rendererClasses [&NiagaraSpriteRendererProperties]}
  {
   emitterName=FlashingElectricity
   rendererClasses [
    &NiagaraLightRendererProperties
    &NiagaraMeshRendererProperties
    &NiagaraSpriteRendererProperties
   ]
  }
 ]
 userVariables [
  User.bIsFrontEnd {type=NiagaraBool}
  User.bIsFrontEndPreview {type=NiagaraBool}
  User.bKillReactivity {type=NiagaraBool, defaultValue=true}
  User.MeshFlickersPerSecond {type=NiagaraFloat, defaultValue=6, description=""}
  User.ShocksPerSecond {type=NiagaraFloat, defaultValue=3, description=""}
 ]
}
