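# Physics asset for the Rumble backpack cosmetic, defining capsule bodies and joints for the physics-driven motion of its arms, legs and hair.
# A small body capsule anchors every chain: left and right three-bone arms and legs, plus eight hair chains (a to h) of two or three bones.
# Each segment is constrained to its parent with narrow swing limits, with limbs allowed less twist than hair.
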PhysicsAsset /BRCosmetics/Accessories/FORT_Backpacks/Backpack_M_MED_Rumble/Meshes/M_MED_Rumble_Pack_Physics
bodies {
 arm_01_l [{shapeType=Capsule, length=2.9401, radius=2.0599, shapeName=arm_01_l_capsule, center=<-2 0 0>, rotation=<rot 90 0 0>}]
 arm_01_r [{shapeType=Capsule, length=2, radius=3, shapeName=arm_01_r_capsule, center=<2 0 0>}]
 arm_02_l [{shapeType=Capsule, length=2.9401, radius=2.0599, shapeName=arm_02_l_capsule, center=<-2 0 0>, rotation=<rot 90 0 0>}]
 arm_02_r [{shapeType=Capsule, length=2, radius=3, shapeName=arm_02_r_capsule, center=<2 0 0>}]
 arm_03_l [{shapeType=Capsule, length=2.9401, radius=2.0599, shapeName=arm_03_l_capsule, center=<-2 0 0>, rotation=<rot 90 0 0>}]
 arm_03_r [{shapeType=Capsule, length=2, radius=3, shapeName=arm_03_r_capsule, center=<2 0 0>}]
 body [{shapeType=Capsule, length=1, radius=1, shapeName=body_capsule}]
 hair_a_01 [{shapeType=Capsule, length=2, radius=3.3714, shapeName=hair_a_01_capsule, center=<2 0 0>, rotation=<rot 90 0 0>}]
 hair_a_02 [{shapeType=Capsule, length=2, radius=3.3714, shapeName=hair_a_02_capsule, center=<2 0 0>, rotation=<rot 90 0 0>}]
 hair_a_03 [{shapeType=Capsule, length=2, radius=3.3714, shapeName=hair_a_03_capsule, center=<2 0 0>, rotation=<rot 90 0 0>}]
 hair_b_01 [{shapeType=Capsule, length=1.2716, radius=1.7666, shapeName=hair_b_01_capsule, center=<1.7146 0 0>, rotation=<rot 90 0 0>}]
 hair_b_02 [{shapeType=Capsule, length=1.2716, radius=1.7666, shapeName=hair_b_02_capsule, center=<1.7146 0 0>, rotation=<rot 90 0 0>}]
 hair_c_01 [{shapeType=Capsule, length=1.1443, radius=1.0933, shapeName=hair_c_01_capsule, center=<2 0 0>, rotation=<rot 90 0 0>}]
 hair_c_02 [{shapeType=Capsule, length=1.1443, radius=1.0933, shapeName=hair_c_02_capsule, center=<2 0 0>, rotation=<rot 90 0 0>}]
 hair_d_01 [{shapeType=Capsule, length=1.2654, radius=1.1853, shapeName=hair_d_01_capsule, center=<2 0 0>, rotation=<rot 90 0 0>}]
 hair_d_02 [{shapeType=Capsule, length=1.2654, radius=1.1853, shapeName=hair_d_02_capsule, center=<2 0 0>, rotation=<rot 90 0 0>}]
 hair_e_01 [{shapeType=Capsule, length=2.1162, radius=1.5087, shapeName=hair_e_01_capsule, center=<2.573 0 0>, rotation=<rot 90 0 0>}]
 hair_e_02 [{shapeType=Capsule, length=2.1162, radius=1.5087, shapeName=hair_e_02_capsule, center=<2.573 0 0>, rotation=<rot 90 0 0>}]
 hair_f_01 [{shapeType=Capsule, length=2, radius=3.3714, shapeName=hair_f_01_capsule, center=<2 0 0>, rotation=<rot 90 0 0>}]
 hair_f_02 [{shapeType=Capsule, length=2, radius=3.3714, shapeName=hair_f_02_capsule, center=<2 0 0>, rotation=<rot 90 0 0>}]
 hair_g_01 [{shapeType=Capsule, length=2, radius=3.3714, shapeName=hair_g_01_capsule, center=<2 0 0>, rotation=<rot 90 0 0>}]
 hair_g_02 [{shapeType=Capsule, length=2, radius=3.3714, shapeName=hair_g_02_capsule, center=<2 0 0>, rotation=<rot 90 0 0>}]
 hair_h_01 [{shapeType=Capsule, length=0.9258, radius=1.8009, shapeName=hair_h_01_capsule, center=<2 0 0>, rotation=<rot 90 0 0>}]
 hair_h_02 [{shapeType=Capsule, length=0.9258, radius=1.8009, shapeName=hair_h_02_capsule, center=<2 0 0>, rotation=<rot 90 0 0>}]
 hair_h_03 [{shapeType=Capsule, length=0.9258, radius=1.8009, shapeName=hair_h_03_capsule, center=<2 0 0>, rotation=<rot 90 0 0>}]
 leg_01_l [{shapeType=Capsule, length=3.0512, radius=2.3574, shapeName=leg_01_l_capsule, center=<-4.1209 0 0>, rotation=<rot 90 0 0>}]
 leg_01_r [{shapeType=Capsule, length=0.2502, radius=4.7498, shapeName=leg_01_r_capsule, center=<3.6737 0 0>, rotation=<rot 90 0 0>}]
 leg_02_l [{shapeType=Capsule, length=3.0512, radius=2.3574, shapeName=leg_02_l_capsule, center=<-4.1209 0 0>, rotation=<rot 90 0 0>}]
 leg_02_r [{shapeType=Capsule, length=0.2502, radius=4.7498, shapeName=leg_02_r_capsule, center=<3.6737 0 0>, rotation=<rot 90 0 0>}]
 leg_03_l [{shapeType=Capsule, length=3.0512, radius=2.3574, shapeName=leg_03_l_capsule, center=<-4.1209 0 0>, rotation=<rot 90 0 0>}]
 leg_03_r [{shapeType=Capsule, length=0.2502, radius=4.7498, shapeName=leg_03_r_capsule, center=<3.6737 0 0>, rotation=<rot 90 0 0>}]
}
constraints [
 {bone1Name=leg_01_r, bone2Name=body, swing1LimitDegrees=15, swing2LimitDegrees=15, twistLimitDegrees=5}
 {bone1Name=leg_02_r, bone2Name=leg_01_r, swing1LimitDegrees=15, swing2LimitDegrees=15, twistLimitDegrees=5}
 {bone1Name=leg_03_r, bone2Name=leg_02_r, swing1LimitDegrees=15, swing2LimitDegrees=15, twistLimitDegrees=5}
 {bone1Name=leg_01_l, bone2Name=body, swing1LimitDegrees=15, swing2LimitDegrees=15, twistLimitDegrees=5}
 {bone1Name=leg_02_l, bone2Name=leg_01_l, swing1LimitDegrees=15, swing2LimitDegrees=15, twistLimitDegrees=5}
 {bone1Name=leg_03_l, bone2Name=leg_02_l, swing1LimitDegrees=15, swing2LimitDegrees=15, twistLimitDegrees=5}
 {bone1Name=arm_01_l, bone2Name=body, swing1LimitDegrees=15, swing2LimitDegrees=15, twistLimitDegrees=5}
 {bone1Name=arm_02_l, bone2Name=arm_01_l, swing1LimitDegrees=15, swing2LimitDegrees=15, twistLimitDegrees=5}
 {bone1Name=arm_03_l, bone2Name=arm_02_l, swing1LimitDegrees=15, swing2LimitDegrees=15, twistLimitDegrees=5}
 {bone1Name=arm_01_r, bone2Name=body, swing1LimitDegrees=15, swing2LimitDegrees=15, twistLimitDegrees=5}
 {bone1Name=arm_02_r, bone2Name=arm_01_r, swing1LimitDegrees=15, swing2LimitDegrees=15, twistLimitDegrees=5}
 {bone1Name=arm_03_r, bone2Name=arm_02_r, swing1LimitDegrees=15, swing2LimitDegrees=15, twistLimitDegrees=5}
 {bone1Name=hair_a_01, bone2Name=body, swing1LimitDegrees=15, swing2LimitDegrees=15, twistLimitDegrees=15}
 {bone1Name=hair_a_02, bone2Name=hair_a_01, swing1LimitDegrees=15, swing2LimitDegrees=15, twistLimitDegrees=15}
 {bone1Name=hair_a_03, bone2Name=hair_a_02, swing1LimitDegrees=15, swing2LimitDegrees=15, twistLimitDegrees=15}
 {bone1Name=hair_b_01, bone2Name=body, swing1LimitDegrees=15, swing2LimitDegrees=15, twistLimitDegrees=15}
 {bone1Name=hair_b_02, bone2Name=hair_b_01, swing1LimitDegrees=15, swing2LimitDegrees=15, twistLimitDegrees=15}
 {bone1Name=hair_c_01, bone2Name=body, swing1LimitDegrees=15, swing2LimitDegrees=15, twistLimitDegrees=15}
 {bone1Name=hair_c_02, bone2Name=hair_c_01, swing1LimitDegrees=15, swing2LimitDegrees=15, twistLimitDegrees=15}
 {bone1Name=hair_d_01, bone2Name=body, swing1LimitDegrees=15, swing2LimitDegrees=15, twistLimitDegrees=15}
 {bone1Name=hair_d_02, bone2Name=hair_d_01, swing1LimitDegrees=15, swing2LimitDegrees=15, twistLimitDegrees=15}
 {bone1Name=hair_e_01, bone2Name=body, swing1LimitDegrees=15, swing2LimitDegrees=15, twistLimitDegrees=15}
 {bone1Name=hair_e_02, bone2Name=hair_e_01, swing1LimitDegrees=15, swing2LimitDegrees=15, twistLimitDegrees=15}
 {bone1Name=hair_f_01, bone2Name=body, swing1LimitDegrees=15, swing2LimitDegrees=15, twistLimitDegrees=15}
 {bone1Name=hair_f_02, bone2Name=hair_f_01, swing1LimitDegrees=15, swing2LimitDegrees=15, twistLimitDegrees=15}
 {bone1Name=hair_g_01, bone2Name=body, swing1LimitDegrees=15, swing2LimitDegrees=15, twistLimitDegrees=15}
 {bone1Name=hair_g_02, bone2Name=hair_g_01, swing1LimitDegrees=15, swing2LimitDegrees=15, twistLimitDegrees=15}
 {bone1Name=hair_h_01, bone2Name=body, swing1LimitDegrees=15, swing2LimitDegrees=15, twistLimitDegrees=15}
 {bone1Name=hair_h_02, bone2Name=hair_h_01, swing1LimitDegrees=15, swing2LimitDegrees=15, twistLimitDegrees=15}
 {bone1Name=hair_h_03, bone2Name=hair_h_02, swing1LimitDegrees=15, swing2LimitDegrees=15, twistLimitDegrees=15}
]
